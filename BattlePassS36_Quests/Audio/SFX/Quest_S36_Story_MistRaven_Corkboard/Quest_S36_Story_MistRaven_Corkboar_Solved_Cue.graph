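# Sound cue for the corkboard "Solved" sound in the Season 36 MistRaven story quest.
# A single wave player feeds the solved wave straight to the output, with a boosted volume multiplier.
# It uses the critical in-game UI sound class, and its concurrency override allows one instance at a time, blocking new plays while one is active.
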SoundCue /BattlePassS36_Quests/Audio/SFX/Quest_S36_Story_MistRaven_Corkboard/Quest_S36_Story_MistRaven_Corkboar_Solved_Cue
root=WavePlayer_0
nodes [{
  id=WavePlayer_0
  properties {soundWaveAssetPtr=/BattlePassS36_Quests/Audio/SFX/Quest_S36_Story_MistRaven_Corkboard/Waves/Quest_S36_Story_MistRaven_Corkboard_Solved}
 }]
properties {
 bOverrideConcurrency=true
 maxDistance=2097152
 soundClassObject=&Fort_UI_InGame_Critical
 volumeMultiplier=2
 concurrencyOverrides {maxCount=1, resolutionRule=PreventNew}
}
