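# Sound cue for collecting the MistRaven corkboard item in a Season 36 story quest.
# A mixer layers one of three random collect waves, pitch/volume-modulated and boosted, with a reward-screen chest upgrade sting after a short fixed delay.
# It plays in the non-critical in-game UI sound class with overridden attenuation and a one-instance concurrency limit that stops the oldest.
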SoundCue /BattlePassS36_Quests/Audio/SFX/Quest_S36_Story_MistRaven_Corkboard/Quest_S36_Story_MistRaven_Corkboard_Collect_Cue
root=Mixer_0
nodes [
 {id=Delay_0, children [WavePlayer_5], properties {delayMax=0.1, delayMin=0.1}}
 {id=Mixer_0, children [Modulator_0,Delay_0], properties {inputVolume [3,1]}}
 {id=Modulator_0, children [Random_0]}
 {id=Random_0, children [WavePlayer_1,WavePlayer_2,WavePlayer_3], properties {weights [1,1,1]}}
 {
  id=WavePlayer_1
  properties {soundWaveAssetPtr=/BattlePassS36_Quests/Audio/SFX/Quest_S36_Story_MistRaven_Corkboard/Waves/Quest_S36_Story_MistRaven_Corkboard_Collect_01}
 }
 {
  id=WavePlayer_2
  properties {soundWaveAssetPtr=/BattlePassS36_Quests/Audio/SFX/Quest_S36_Story_MistRaven_Corkboard/Waves/Quest_S36_Story_MistRaven_Corkboard_Collect_02}
 }
 {
  id=WavePlayer_3
  properties {soundWaveAssetPtr=/BattlePassS36_Quests/Audio/SFX/Quest_S36_Story_MistRaven_Corkboard/Waves/Quest_S36_Story_MistRaven_Corkboard_Collect_03}
 }
 {id=WavePlayer_5, properties {soundWaveAssetPtr=/Game/Sounds/Fort_UI_Sounds/Rewards/Fort_RewardScreen_TeamScore_ChestLootUpgrade}}
]
properties {
 bOverrideAttenuation=true
 bOverrideConcurrency=true
 maxDistance=1250
 soundClassObject=&Fort_UI_InGame_NonCritical
 attenuationOverrides {falloffDistance=850, stereoSpread=2300}
 concurrencyOverrides {maxCount=1, resolutionRule=StopOldest}
}
